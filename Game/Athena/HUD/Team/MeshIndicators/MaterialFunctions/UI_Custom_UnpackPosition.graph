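# UI_Custom_UnpackPosition is a material function for team HUD mesh indicators that supplies a packed position to materials.
# It reads texture coordinate channel 2, masks it to red and green, and returns that two-component value as its only output.
# The name indicates that this UV channel carries position data, and the function applies no scaling or other decoding.
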MaterialFunction /Game/Athena/HUD/Team/MeshIndicators/MaterialFunctions/UI_Custom_UnpackPosition
graph {nodes [FunctionOutput(None=ComponentMask(None=TextureCoordinate() {coordinateIndex=2}) {g=true, r=true})]}
