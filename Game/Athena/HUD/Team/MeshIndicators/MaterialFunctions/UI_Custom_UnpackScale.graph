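# UI_Custom_UnpackScale is a material function for the team mesh indicator HUD that unpacks a scale value.
# It reads texture coordinate channel 3, masks out the red component and returns it as its single output.
# This suggests per-element scale is packed into the U of that UV set, though the file does not show what writes it.
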MaterialFunction /Game/Athena/HUD/Team/MeshIndicators/MaterialFunctions/UI_Custom_UnpackScale
graph {nodes [FunctionOutput(None=ComponentMask(None=TextureCoordinate() {coordinateIndex=3}) {r=true})]}
